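# Niagara swing trail for the SkeleProbe pickaxe, built from four animation-trail ribbon emitters: thin, faint, sharp and galaxy.
# Each spawns from weapon bones or sockets, shrinks over a short lifetime and adjusts alpha for time of day.
# User.Variant indexes color arrays in the thin, faint and sharp trails.
# Emitter bools derived from Variant are also set, including a variant-5 flag on galaxy and four on sharp, which has four ribbon renderers.
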NiagaraSystem /BRCosmetics/Weapons/FORT_Melee/Pickaxe_SkeleProbe/FX/NS_Pickaxe_SkeleProbe_Trail
dependencies {
 usedDataInterfaces [&NiagaraDataInterfaceCurve, &NiagaraDataInterfaceArrayColor]
 usedDynamicInputs []
 usedModules [
  &SystemState
  &self:SystemSpawnScript.SystemScriptSource.SystemScriptGraph.NiagaraNodeAssignment_0.SetVariables_#1
 ]
}
emitters {
 Thin001 {
  emitterUpdateScript [
   &EmitterState()
   &FN_SpawnAnimTrail_v2(AnimTrailWidthMode=::FromFirst)
  ]
  particleSpawnScript [
   /Game/Effects/Niagara/Modules/AnimTrail/FN_InitializeAnimTrail/FN_InitializeAnimTrailBoneIndecies(First Bone or Socket Index=1)
   /Game/Effects/Niagara/Modules/AnimTrail/FN_InitializeAnimTrail/FN_InitializeAnimTrail(
    AnimTrailWidth=0.05
    Color=&SelectLinearColorFromArray(
     Array Sampling Mode=::NewEnumerator1
     Direct Array Index=&MakeIntFromFloat(Float=$User.Variant)
     Color Selection Array=DataInterface() {propertyValues {ColorData [
        <rgba 0.5852 1.228 1.5 1>
        <rgba 1.35 0.54 1.1559 1>
        <rgba 1.35 0.5772 0.0915 1>
        <rgba 0.28 1.12 1.4 1>
        <rgba 0.4998 1.2747 1.666 1>
        <rgba 0.375 0.09 1.8 1>
       ]}}
     Random Seed=?
    )
    Lifetime=0.1
   )
   &SystemLocation(MaskSpawnGroup=false)
  ]
  particleUpdateScript [
   &ScaleRibbonWidth(Ribbon Width Scale=&FloatFromCurve(FloatCurve=DataInterface() {propertyValues {Curve {keys [
         {arriveTangent=-1.2579, interpMode=RCIM_Cubic, time=0, value=1}
         {arriveTangent=-0.9664, interpMode=RCIM_Cubic, leaveTangent=-0.9664, tangentMode=RCTM_User, time=1, value=0}
        ]}}}))
   &ScaleRibbonWidth(Ribbon Width Scale=&FloatFromCurve(FloatCurve=DataInterface() {propertyValues {Curve {keys [
         {arriveTangent=-1.6294, interpMode=RCIM_Cubic, leaveTangent=-1.6294, tangentMode=RCTM_User, time=0, value=1}
         {interpMode=RCIM_Cubic, tangentMode=RCTM_User, time=1, value=0}
        ]}}})) {module=ScaleRibbonWidth001}
   &ScaleColor(
    Scale Alpha=&FloatFromCurve(
     FloatCurve=DataInterface() {propertyValues {Curve {keys [
         {arriveTangent=-2.6908, interpMode=RCIM_Cubic, leaveTangent=-2.6908, tangentMode=RCTM_User, time=0.1, value=1}
         {arriveTangent=-0.2312, interpMode=RCIM_Cubic, leaveTangent=-0.2312, tangentMode=RCTM_User, time=0.5536, value=0.0233}
         <at 0.8036 0 RCIM_Cubic>
        ]}}}
     Scale Curve=&Multiply_Float(A=1, B=&FNTimeOfDayScalar(Night Scale=0.1, Morning Scale=0.35, Evening Scale=0.35))
    )
    ScaleA=true
    ScaleRGB=false
   ) {module=ScaleColor001}
   &FN_UpdateAnimTrail_v2()
  ]
  rendererClasses [&NiagaraRibbonRendererProperties]
  renderers [{rendererClass=&NiagaraRibbonRendererProperties}]
 }
 Trail_Faint {
  emitterUpdateScript [
   &EmitterState()
   &FN_SpawnAnimTrail_v2(AnimTrailWidthMode=::FromSecond)
  ]
  particleSpawnScript [
   &FN_InitializeAnimTrailBoneIndecies(Second Bone or Socket Index=1)
   &FN_InitializeAnimTrail(AnimTrailWidth=1, Lifetime=0.125)
   &SystemLocation(MaskSpawnGroup=false)
  ]
  particleUpdateScript [
   &ScaleRibbonWidth(Ribbon Width Scale=&FloatFromCurve(FloatCurve=DataInterface() {propertyValues {Curve {keys [
         {arriveTangent=-1.2579, interpMode=RCIM_Cubic, time=0, value=1}
         {arriveTangent=-0.9664, interpMode=RCIM_Cubic, leaveTangent=-0.9664, tangentMode=RCTM_User, time=1, value=0}
        ]}}}))
   &FN_UpdateAnimTrail_v2()
   &Color(
    Color=&MultiplyLinearColor(
     Linear Color=&SelectLinearColorFromArray(
      Array Sampling Mode=::NewEnumerator1
      Direct Array Index=&MakeIntFromFloat(Float=$User.Variant)
      Color Selection Array=DataInterface() {propertyValues {ColorData [
         <rgba 0.0284 0.0284 0.0409 0.5>
         <rgba 0.9047 0.4852 0.7605 0.125>
         <rgba 0.0953 0.0723 0.0685 1>
         <rgba 0.1981 0.5583 0.6795 0.35>
         <rgba 0.5776 0.5906 0.5972 0.42>
         <rgba 0.0252 0.5647 0.7379 0.15>
        ]}}
      Random Seed=?
     )
     Scale Factor=&MakeVector4(
      X=1
      Y=1
      Z=1
      W=&FloatFromCurve(FloatCurve=DataInterface() {propertyValues {Curve {keys [
           {arriveTangent=-2.6908, interpMode=RCIM_Cubic, leaveTangent=-2.6908, tangentMode=RCTM_User, time=0.1, value=1}
           {arriveTangent=-0.2312, interpMode=RCIM_Cubic, leaveTangent=-0.2312, tangentMode=RCTM_User, time=0.5536, value=0.0233}
           <at 0.8036 0 RCIM_Cubic>
          ]}}})
     )
    )
    Scale Alpha=&Multiply_Float(A=$System.TimeOfDayColorMultiplier, B=1.8)
   )
  ]
  rendererClasses [&NiagaraRibbonRendererProperties]
  renderers [{rendererClass=&NiagaraRibbonRendererProperties}]
 }
 Trail_Galaxy {
  emitterSpawnScript [/SetVariables(Emitter.Variant 5 Bool=HlslExpression() {hlslExpression="(User.Variant==5) ? (true) : (false)"}) {isSetParameters=true, module=SetVariables_#2}]
  emitterUpdateScript [
   &EmitterState()
   &FN_SpawnAnimTrail_v2(AnimTrailWidthMode=::FromSecond)
  ]
  particleSpawnScript [
   &FN_InitializeAnimTrailBoneIndecies(First Bone or Socket Index=1)
   &FN_InitializeAnimTrail(AnimTrailWidth=1, Lifetime=0.125)
   &SystemLocation(MaskSpawnGroup=false)
  ]
  particleUpdateScript [
   &ScaleRibbonWidth(Ribbon Width Scale=&FloatFromCurve(FloatCurve=DataInterface() {propertyValues {Curve {keys [
         {arriveTangent=-2.7553, interpMode=RCIM_Cubic, leaveTangent=-2.7553, tangentMode=RCTM_User, time=0, value=1}
         {arriveTangent=-0.1336, interpMode=RCIM_Cubic, leaveTangent=-0.1336, tangentMode=RCTM_User, time=1, value=0}
        ]}}})) {enabled=false}
   &ScaleRibbonWidth(Ribbon Width Scale=&FloatFromCurve(FloatCurve=DataInterface() {propertyValues {Curve {keys [
         {arriveTangent=-1.2579, interpMode=RCIM_Cubic, time=0, value=1}
         {arriveTangent=-0.9664, interpMode=RCIM_Cubic, leaveTangent=-0.9664, tangentMode=RCTM_User, time=1, value=0}
        ]}}})) {module=ScaleRibbonWidth001}
   &FN_UpdateAnimTrail_v2()
   &Color(Color=<rgba 1 1 1 0.3>, Scale Alpha=$System.TimeOfDayColorMultiplier)
  ]
  rendererClasses [&NiagaraRibbonRendererProperties]
  renderers [{rendererClass=&NiagaraRibbonRendererProperties}]
 }
 Trail_Sharp {
  emitterSpawnScript [/SetVariables(
    Emitter.Main Bool=HlslExpression() {hlslExpression="(User.Variant==0 || User.Variant==3 || User.Variant==5) ? (false) : (true)"}
    Emitter.Variant 0 Bool=HlslExpression() {hlslExpression="(User.Variant==0) ? (true) : (false)"}
    Emitter.Variant 3 Bool=HlslExpression() {hlslExpression="(User.Variant==3) ? (true) : (false)"}
    Emitter.Variant 5 Bool=HlslExpression() {hlslExpression="(User.Variant==5) ? (true) : (false)"}
   ) {isSetParameters=true, module=SetVariables_#2}]
  emitterUpdateScript [
   &EmitterState()
   &FN_SpawnAnimTrail_v2(AnimTrailWidthMode=::FromSecond)
  ]
  particleSpawnScript [
   &FN_InitializeAnimTrailBoneIndecies(Second Bone or Socket Index=1)
   &FN_InitializeAnimTrail(AnimTrailWidth=0.6, Lifetime=0.1)
   &SystemLocation(MaskSpawnGroup=false)
  ]
  particleUpdateScript [
   &ScaleRibbonWidth(Ribbon Width Scale=&FloatFromCurve(FloatCurve=DataInterface() {propertyValues {Curve {keys [
         {arriveTangent=-1.2579, interpMode=RCIM_Cubic, time=0, value=1}
         {arriveTangent=-0.9664, interpMode=RCIM_Cubic, leaveTangent=-0.9664, tangentMode=RCTM_User, time=1, value=0}
        ]}}})) {module=ScaleRibbonWidth001}
   &FN_UpdateAnimTrail_v2()
   &Color(
    Color=&MultiplyLinearColor(
     Linear Color=&SelectLinearColorFromArray(
      Array Sampling Mode=::NewEnumerator1
      Direct Array Index=&MakeIntFromFloat(Float=$User.Variant)
      Color Selection Array=DataInterface() {propertyValues {ColorData [
         <rgba 1 1 1 0.8>
         <rgba 0.7758 0.0006 0.521 0.5>
         <rgba 0.5 0.282 0.064 1>
         <rgba 1 1 1 0.777>
         <rgba 0.0003 0.3564 0.6795 1>
         <rgba 1 1 1 2>
        ]}}
      Random Seed=?
     )
     Scale Factor=&MakeVector4(
      X=1
      Y=1
      Z=1
      W=&FloatFromCurve(FloatCurve=DataInterface() {propertyValues {Curve {keys [
           {arriveTangent=-2.6908, interpMode=RCIM_Cubic, leaveTangent=-2.6908, tangentMode=RCTM_User, time=0.1, value=1}
           {arriveTangent=-0.2312, interpMode=RCIM_Cubic, leaveTangent=-0.2312, tangentMode=RCTM_User, time=0.5536, value=0.0233}
           <at 0.8036 0 RCIM_Cubic>
          ]}}})
     )
    )
    Scale Alpha=&Multiply_Float(A=$System.TimeOfDayColorMultiplier, B=1.8)
   )
  ]
  rendererClasses [&NiagaraRibbonRendererProperties]
  renderers [
   {rendererClass=&NiagaraRibbonRendererProperties}
   {rendererClass=&NiagaraRibbonRendererProperties, rendererIndex=1}
   {rendererClass=&NiagaraRibbonRendererProperties, rendererIndex=2}
   {rendererClass=&NiagaraRibbonRendererProperties, rendererIndex=3}
  ]
 }
}
summary {
 systemName=NS_Pickaxe_SkeleProbe_Trail
 emitters [
  {emitterName=Trail_Faint, rendererClasses [&NiagaraRibbonRendererProperties]}
  {emitterName=Trail_Sharp, rendererClasses [&NiagaraRibbonRendererProperties]}
  {emitterName=Thin001, rendererClasses [&NiagaraRibbonRendererProperties]}
  {emitterName=Trail_Galaxy, rendererClasses [&NiagaraRibbonRendererProperties]}
 ]
 userVariables [
  User.RibbonIDOffset {type=NiagaraInt32}
  User.SkeletalMesh {
   type=NiagaraDataInterfaceSkeletalMesh
   defaultValue {
    struct=NiagaraExt_VariableValue_DataInterface
    value {dataInterface=&self:NiagaraDataInterfaceSkeletalMesh_0}
   }
  }
  User.Variant {type=NiagaraFloat, defaultValue=0, description=""}
 ]
}
